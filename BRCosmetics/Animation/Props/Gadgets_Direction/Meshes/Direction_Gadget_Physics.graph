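# Physics asset for the Direction Gadget prop in BRCosmetics, giving its skeletal mesh simple collision shapes.
# A sphere body covers the root bone, and a rotated capsule body covers the JNT_SpinnerSign joint, which by its name is a spinning sign.
# No constraints link the two bodies, so the asset defines no physical joint between them.
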PhysicsAsset /BRCosmetics/Animation/Props/Gadgets_Direction/Meshes/Direction_Gadget_Physics
bodies {
 JNT_SpinnerSign [{shapeType=Capsule, length=56.6208, radius=14.2827, center=<0.1014 0.0056 0.0001>, rotation=<rot 0 90.0018 -89.9993>}]
 root [{shapeType=Sphere, radius=15, shapeName=root_sphere}]
}
constraints []
